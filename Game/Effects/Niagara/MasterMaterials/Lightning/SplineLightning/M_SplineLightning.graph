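# Translucent unlit, two-sided master material that draws animated lightning on Niagara spline mesh particles.
# Two panned noise samples distort UVs for a sharp core band and blurred glow, feeding opacity and color-tinted emissive.
# Dynamic parameters drive width, growth, dissolve and motion; base/tip controls shape the ends, and depth fade skips Low/Medium quality.
# World position offset combines SplineThicken with sine oscillation, optionally masked to pin either end.
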Material /Game/Effects/Niagara/MasterMaterials/Lightning/SplineLightning/M_SplineLightning
graph {
 properties {
  bCastRayTracedShadows=false
  bUsedWithNiagaraMeshParticles=true
  blendMode=BLEND_Translucent
  pixelDepthOffsetMode=PDOM_Legacy
  shadingModel=MSM_Unlit
  twoSided=true
  usedShadingModels=Unlit
 }
 nodes [
  n1 = TextureCoordinate() {vTiling=2}
  n2 = ScalarParameter() {defaultValue=1, group=WPO Oscillation, parameterName=Sin Frequency, sliderMax=5, sortPriority=1}
  n3 = DynamicParameter() {parameterIndex=1, defaultValue=<rgba 0 1 1 1>, paramNames [Dissolve, Noise Frequency, Sin Frequency, Noise Speed]}
  n4 = Add(A=AppendVector(A=Constant() {r=1}, B=Multiply(A=@n2, B=@n3.Sin Frequency)), B=ParticleRandom())
  NamedRerouteDeclaration(None=ComponentMask(None=Multiply(A=@n1, B=@n4)) {g=true}) {name=UV.v, nodeColor=<rgba 1 0 0.7266 1>}
  n5 = ParticleColor()
  NamedRerouteDeclaration(None=@n5.A) {name=Particle Alpha, nodeColor=<rgba 0.8984 1 0 1>}
  n6 = ScalarParameter() {
   defaultValue=1
   desc="Dampens the WPO oscillation at the tip. This allows you to \"pin\" the tip of the lightning so that the tip doesn't move around."
   group=WPO Oscillation
   parameterName=Tip WPO Mask Scale
   sortPriority=41
  }
  n7 = NamedRerouteUsage()
  n8 = ComponentMask(None=@n7) {r=true}
  n9 = ScalarParameter() {
   defaultValue=1
   desc="Dampens the WPO oscillation at the base. This allows you to \"pin\" the base of the lightning so that the base doesn't move around. Think Emperor Palpatine's fingertips."
   group=WPO Oscillation
   parameterName=Base WPO Mask Scale
   sortPriority=43
  }
  n10 = ComponentMask(None=@n7) {g=true}
  NamedRerouteDeclaration(None=Multiply(A=LinearInterpolate(A=@n6, Alpha=@n8), B=LinearInterpolate(A=@n9, Alpha=@n10))) {name=WPO Sin Mask, nodeColor=<rgba 1 0.1725 0.0207 1>}
  n11 = ScalarParameter() {group=Base to Tip, parameterName=Tip Width Scalar, sliderMax=10, sortPriority=3}
  n12 = NamedRerouteUsage()
  n13 = ComponentMask(None=@n12) {r=true}
  n14 = ScalarParameter() {group=Base to Tip, parameterName=Base Width Scalar, sliderMax=10, sortPriority=1}
  n15 = ComponentMask(None=@n12) {g=true}
  NamedRerouteDeclaration(None=Multiply(A=LinearInterpolate(A=@n11, Alpha=@n13), B=LinearInterpolate(A=@n14, Alpha=@n15))) {name=UV.V Scalar, nodeColor=<rgba 1 0.3984 0 1>}
  n16 = NamedRerouteUsage()
  n17 = Divide(A=@n16, B=ScalarParameter() {defaultValue=0.1, group=Base to Tip, parameterName=Tip Mask Spread, sliderMax=0.5, sortPriority=1})
  n18 = Divide(
   A=OneMinus(None=@n16)
   B=ScalarParameter() {defaultValue=0.1, group=Base to Tip, parameterName=Base Mask Spread, sliderMax=0.5, sortPriority=0}
  )
  NamedRerouteDeclaration(None=AppendVector(A=Saturate(None=@n17), B=Saturate(None=@n18))) {name=BaseTipMask, nodeColor=<rgba 0.0547 1 0 1>}
  n19 = TextureCoordinate()
  NamedRerouteDeclaration(None=ComponentMask(None=@n19) {r=true}) {name=TexCoord.u, nodeColor=<rgba 1 0 0.9141 1>}
  n20 = ComponentMask(None=@n19) {g=true}
  NamedRerouteDeclaration(None=StaticSwitchParameter(True=OneMinus(None=@n20), False=@n20) {group=Base to Tip, parameterName=Reverse Base and Tip, sortPriority=-10}) {name=TexCoord.v, nodeColor=<rgba 1 0 0.9141 1>}
  n21 = NamedRerouteUsage()
  n22 = NamedRerouteUsage()
  n23 = Multiply(
   A=DynamicParameter() {paramNames [Width Scalar, Sin Amplitude, Opacity, Growth]}.Width Scalar
   B=LinearInterpolate(
    A=ScalarParameter() {group=Lightning Noise, parameterName=Lightning Width Min, sliderMax=1, sortPriority=5}
    B=ScalarParameter() {defaultValue=0.05, group=Lightning Noise, parameterName=Lightning Width Max, sliderMax=1, sortPriority=6}
    Alpha=Saturate(None=@n22)
   ) {constB=0.1}
  )
  n24 = NamedRerouteUsage()
  n25 = AppendVector(A=NamedRerouteUsage(), B=Add(A=NamedRerouteUsage(), B=ParticleRandom()))
  n26 = Divide(B=ScalarParameter() {defaultValue=0.08, group=Lightning Noise, parameterName=Noise U Frequency, sliderMax=1, sortPriority=1})
  n27 = DynamicParameter() {parameterIndex=1, defaultValue=<rgba 0 1 1 1>, paramNames [Dissolve, Noise Frequency, Sin Frequency, Noise Speed]}
  n28 = Divide(B=Multiply(
    A=ScalarParameter() {defaultValue=0.55, group=Lightning Noise, parameterName=Noise V Frequency, sliderMax=10, sortPriority=2}
    B=@n27.Noise Frequency
   ))
  n29 = ScalarParameter() {defaultValue=0.25, group=Lightning Noise, parameterName=Noise Speed, sortPriority=4}
  n30 = Panner(
   Coordinate=MaterialFunctionCall(UVs=@n25, Texture Scale=AppendVector(A=@n26, B=@n28)) {materialFunction=&ScaleUVsByCenter}.UVs
   Speed=AppendVector(A=Constant(), B=Multiply(A=@n29, B=@n27.Noise Speed))
  )
  n31 = TextureObjectParameter() {
   desc=Texture for adding noise to the lightning. Recommend that sRGB be off for centered values.
   group=Lightning Noise
   parameterName=Noise Texture
   samplerType=SAMPLERTYPE_LinearColor
   sortPriority=0
   texture=&/T_SplineLightning_Noise
  }
  n32 = TextureSample(UVs=@n30, Tex=@n31) {
   samplerSource=SSM_Wrap_WorldGroupSettings
   samplerType=SAMPLERTYPE_LinearColor
   texture=&T_SplineLightning_Noise
  }
  n33 = NamedRerouteDeclaration(None=@n32.R) {name=NoiseCore, nodeColor=<rgba 0.4531 0 1 1>}
  n34 = AppendVector(A=@n33, B=@n33)
  n35 = Multiply(A=Subtract(A=Saturate(None=@n34)) {constB=0.5}) {constB=2}
  n36 = ScalarParameter() {defaultValue=0.125, group=Lightning Noise, parameterName=Noise Amplitude, sliderMax=1, sortPriority=3}
  n37 = TextureSample(UVs=Multiply(A=@n30) {constB=2}, Tex=@n31) {
   samplerSource=SSM_Wrap_WorldGroupSettings
   samplerType=SAMPLERTYPE_LinearColor
   texture=&T_SplineLightning_Noise
  }
  n38 = AppendVector(A=@n37.R, B=@n37.R)
  n39 = Multiply(A=Subtract(A=Saturate(None=@n38)) {constB=0.5}) {constB=2}
  n40 = Multiply(A=ScalarParameter() {defaultValue=0.125, group=Lightning Noise, parameterName=Noise Amplitude, sliderMax=1, sortPriority=3}) {constB=0.5}
  n41 = Add(A=@n25, B=Add(A=Multiply(A=@n35, B=@n36), B=Multiply(A=@n39, B=@n40)))
  n42 = NamedRerouteUsage()
  n43 = Multiply(
   A=MaterialFunctionCall!(
    Width=Multiply(A=@n23, B=@n24)
    Sharpness=ScalarParameter() {defaultValue=0.5, group=Lightning, parameterName=Lightning Sharpness, sliderMax=1, sortPriority=4}
    Offset
    Direction Switch=StaticBool() {value=true}
    Compare
    Input Coordinates=@n41
   ) {materialFunction=/Engine/Functions/Engine_MaterialFunctions03/Procedurals/GeneratedBand}.Result
   B=LinearInterpolate(
    A=ScalarParameter() {group=Lightning Noise, parameterName=Lightning Opacity Min, sliderMax=1, sortPriority=7}
    B=ScalarParameter() {defaultValue=1, group=Lightning Noise, parameterName=Lightning Opacity Max, sliderMax=1, sortPriority=8}
    Alpha=Saturate(None=@n42)
   ) {constB=0.1}
  )
  n44 = Multiply(
   A=Multiply(
    A=DynamicParameter() {paramNames [Width Scalar, Sin Amplitude, Opacity, Growth]}.Width Scalar
    B=ScalarParameter() {defaultValue=0.175, group=Lightning Blur, parameterName=Blur Width, sliderMax=1, sortPriority=0}
   )
   B=NamedRerouteUsage()
  )
  n45 = ScalarParameter() {group=Lightning Blur, parameterName=Blur Sharpness, sliderMax=1, sortPriority=1}
  n46 = StaticBool() {value=true}
  n47 = Max(
   A=@n43
   B=Multiply(
    A=MaterialFunctionCall(Width=@n44, Sharpness=@n45, Direction Switch=@n46, Input Coordinates=@n41) {materialFunction=&GeneratedBand}.Result
    B=ScalarParameter() {defaultValue=0.2, group=Lightning Blur, parameterName=Blur Opacity, sortPriority=2}
   )
  )
  n48 = TextureCoordinate()
  n49 = DynamicParameter() {paramNames [Width Scalar, Sin Amplitude, Opacity, Growth]}
  n50 = Multiply(
   A=Multiply(A=Multiply(A=@n21, B=@n47), B=DynamicParameter() {paramNames [Width Scalar, Sin Amplitude, Opacity, Growth]}.Opacity)
   B=Add(A=ComponentMask(None=@n48) {g=true}, B=LinearInterpolate(Alpha=@n49.Growth) {constA=-1})
  )
  n51 = DynamicParameter() {parameterIndex=1, defaultValue=<rgba 0 1 1 1>, paramNames [Dissolve, Noise Frequency, Sin Frequency, Noise Speed]}
  n52 = OneMinus(None=Saturate(None=@n50))
  n53 = ScalarParameter() {group=Opacity, parameterName=Dissolve Contrast}
  n54 = ScalarParameter() {defaultValue=1, desc=Scale the opacity at the tip., group=Opacity, parameterName=Tip Opacity Scalar, sliderMax=1, sortPriority=2}
  n55 = NamedRerouteUsage()
  n56 = ComponentMask(None=@n55) {r=true}
  n57 = ScalarParameter() {defaultValue=1, desc=Scale the opacity at the base., group=Opacity, parameterName=Base Opacity Scalar, sliderMax=1, sortPriority=3}
  n58 = ComponentMask(None=@n55) {g=true}
  n59 = Multiply(
   A=Multiply(
    A=@n50
    B=MaterialFunctionCall(Transition Phase=@n51.Dissolve, Height Texture=@n52, Contrast=@n53) {materialFunction=&HeightLerp}.Results
   )
   B=Multiply(A=LinearInterpolate(A=@n54, Alpha=@n56), B=LinearInterpolate(A=@n57, Alpha=@n58))
  )
  n60 = DepthFade(Opacity=@n59, FadeDistance=ScalarParameter() {defaultValue=10, group=Opacity, parameterName=FadeDistance, sortPriority=0})
  Reroute(None=QualitySwitch(Default=@n60, Low=@n59, High=@n60, Medium=@n59, Epic=@n60)) {desc=OPACITY}
  n61 = ScalarParameter() {defaultValue=20, desc=Global spline width scalar., group=WPO Spline, parameterName=Spline Width, sortPriority=2}
  n62 = ScalarParameter() {defaultValue=20, group=WPO Spline, parameterName=Spline Tip Width}
  n63 = Time()
  n64 = Add(A=@n63) {constB=0}
  n65 = ScalarParameter() {defaultValue=1, group=WPO Oscillation, parameterName=Sin Speed, sortPriority=0}
  n66 = Sine(None=Add(A=NamedRerouteUsage(), B=Multiply(A=@n64, B=@n65) {constB=2}))
  n67 = Add(A=@n63) {constB=0.1}
  n68 = Sine(None=Add(A=NamedRerouteUsage(), B=Multiply(A=@n67, B=@n65) {constB=2}))
  n69 = Add() {constB=0.2}
  n70 = Sine(None=Add(A=NamedRerouteUsage(), B=Multiply(A=@n69, B=@n65) {constB=2}))
  n71 = Multiply(
   A=Transform(None=MaterialFunctionCall(R=@n66, G=@n68, B=@n70) {materialFunction=&AppendMany}.RGB) {transformSourceType=TRANSFORMSOURCE_Local}
   B=Multiply(
    A=DynamicParameter() {paramNames [Width Scalar, Sin Amplitude, Opacity, Growth]}.Sin Amplitude
    B=ScalarParameter() {defaultValue=1, group=WPO Oscillation, parameterName=Sin Amplitude, sliderMax=10, sortPriority=2}
   )
  )
  n72 = Multiply(A=NamedRerouteUsage(), B=@n71)
  Reroute(None=Add(
    A=MaterialFunctionCall(WidthBase=@n61, WidthTip=@n62) {materialFunction=&SplineThicken}.WorldPositionOffset
    B=StaticSwitchParameter(True=@n72, False=@n71) {group=WPO Oscillation, parameterName=WPO Sin Mask Toggle, sortPriority=40}
   )) {desc=WPO}
  n73 = VectorParameter() {
   desc=Set this value to white when using particle color.
   group=Lightning
   parameterName=Lightning Color
   sortPriority=0
   defaultValue=<rgba 1 1 1 1>
  }
  n74 = Multiply(A=@n73.RGB, B=@n5.RGB)
  n75 = ScalarParameter() {group=Lightning, parameterName=Lightning Color Core Desat, sliderMax=1}
  n76 = Multiply(
   A=Multiply(
    A=Desaturation(None=@n74, Fraction=@n75)
    B=ScalarParameter() {defaultValue=5, group=Lightning, parameterName=Lightning Color Core Scalar, sliderMax=10, sortPriority=1}
   )
   B=@n43
  )
  Reroute(None=MaterialFunctionCall(Emissive=Add(A=@n74, B=@n76)) {materialFunction=&MF_FX_EmissiveAdjustments}.Result) {desc=EMISSIVE COLOR}
 ]
}
